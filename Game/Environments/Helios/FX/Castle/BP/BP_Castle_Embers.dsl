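; Castle ember effect Blueprint that exposes per-instance tuning for the SM_Motes_Few mote mesh.
; The construction script creates a dynamic material instance from the mesh's first material slot.
; It copies actor variables into that material's size, direction, loop duration, hue shift range, curl and HDR intensity parameters.
; BeginPlay, overlap and Tick events are empty, so the Blueprint runs no gameplay logic.
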(event EventBeginPlay)

(event Collision|EventActorBeginOverlap
  (params (OtherActor :type Actor%20Object%20Reference)))

(event EventTick
  (params (DeltaSeconds :type Float%20%28single-precision%29)))

(fn ConstructionScript
  ;[1 1 1] Setup material parameters
  (bind _returnvalue (Rendering|Material|CreateDynamicMaterialInstance (Variables|Default|GetSM_Motes_Few) 0 (Rendering|Material|GetMaterial (Variables|Default|GetSM_Motes_Few) 0) "None" :!class /Script/Engine.PrimitiveComponent))
  (Utilities|FlowControl|Sequence
    (:then_0
      (Rendering|Material|SetScalarParameterValue _returnvalue "Size" (Variables|Default|GetSize) :!class /Script/Engine.MaterialInstanceDynamic)
      (Rendering|Material|SetVectorParameterValue _returnvalue "OverallDirection" (Math|Conversions|ToLinearColor(Vector) (Variables|Default|GetOverallDirection)) :!class /Script/Engine.MaterialInstanceDynamic)
      (Rendering|Material|SetScalarParameterValue _returnvalue "LoopDuration" (Variables|Default|GetLoopDuration) :!class /Script/Engine.MaterialInstanceDynamic))
    (:then_1
      (Rendering|Material|SetScalarParameterValue _returnvalue "HueShiftMin" (Variables|Default|GetHueShiftMin) :!class /Script/Engine.MaterialInstanceDynamic)
      (Rendering|Material|SetScalarParameterValue _returnvalue "HueShiftMax" (Variables|Default|GetHueShiftMax) :!class /Script/Engine.MaterialInstanceDynamic)
      (Rendering|Material|SetScalarParameterValue _returnvalue "Curl" (Variables|Default|GetCurl) :!class /Script/Engine.MaterialInstanceDynamic)
      (Rendering|Material|SetScalarParameterValue _returnvalue "HDRIntensity" (Variables|Default|GetEmissiveIntensity) :!class /Script/Engine.MaterialInstanceDynamic))))
