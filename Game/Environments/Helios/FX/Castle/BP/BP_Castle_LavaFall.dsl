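; Castle lava fall effect Blueprint from the Helios environment, carrying a looping molten-metal forge audio component.
; When the actor is destroyed, nested validity checks gate a short linear fade-out of that loop to silence.
; BeginPlay, overlap, Tick and the construction script are empty, so the audio fade is the only scripted behavior.
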(event EventBeginPlay)

(event Collision|EventActorBeginOverlap
  (params (OtherActor :type Actor%20Object%20Reference)))

(event EventTick
  (params (DeltaSeconds :type Float%20%28single-precision%29)))

(event EventDestroyed
  ;[1 1 1] Event Destroyed - Fade Out Audio
  (Utilities|IsValid 0
    (:Is%20Valid
      (if (Utilities|IsValid 0 :!class /Script/Engine.KismetSystemLibrary)
        (Utilities|IsValid 0
          (:Is%20Valid
            (if (Utilities|IsValid 0 :!class /Script/Engine.KismetSystemLibrary)
              (Utilities|IsValid 0
                (:Is%20Valid
                  (if (Utilities|IsValid 0 :!class /Script/Engine.KismetSystemLibrary)
                    (Audio|Components|Audio|FadeOut (Variables|Default|GetSpot_Forge_MoltenMetal_Top_Loop_Cue) 0.2 0.0 "Linear" :!class /Script/Engine.AudioComponent)))))))))))

(fn ConstructionScript)
