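; BP_Castle_SmokeManhole is an actor Blueprint in the Helios castle environment FX folder, named for a smoking manhole effect.
; Its begin play, actor overlap and tick events and its construction script are all empty, so it adds no Blueprint logic of its own.
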(event EventBeginPlay)

(event Collision|EventActorBeginOverlap
  (params (OtherActor :type Actor%20Object%20Reference)))

(event EventTick
  (params (DeltaSeconds :type Float%20%28single-precision%29)))

(fn ConstructionScript)
