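# Niagara system for falling embers in the Castle environment: one GPU sprite emitter spawning continuously at the User.SpawnRate rate.
# Particles spawn across a wide shape with randomized HDR orange color, size and lifetime, plus downward, sideways and curl-noise initial motion.
# Gravity and panning, cone-masked curl noise drive drift, while sprites rotate, grow then shrink, and feed random values to material parameters.
# User.CylinderHeight and User.CylinderRadius are exposed, with no visible binding.
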NiagaraSystem /Game/Environments/Helios/FX/Castle/Effects/NS_Castle_FallingEmbers
dependencies {
 usedDataInterfaces [&NiagaraDataInterfaceCurve, &NiagaraDataInterfaceArrayColor]
 usedDynamicInputs [&Lerp_Vector2, &Multiply_Float]
 usedModules [
  &SystemState
  &self:SystemSpawnScript.SystemScriptSource.SystemScriptGraph.NiagaraNodeAssignment_0.SetVariables_#1
 ]
}
emitters {Embers_Burning_Cylinder {
  simTarget=GPUComputeSim
  emitterUpdateScript [
   &EmitterState()
   &SpawnRate(SpawnRate=$User.SpawnRate, Use Spawn Probability=false)
  ]
  particleSpawnScript [
   /SetVariables(Particles.RandSpreadCurve:NiagaraFloat=&FloatFromCurve(
     FloatCurve=DataInterface() {propertyValues {Curve {keys [
         {interpMode=RCIM_Cubic, tangentMode=RCTM_User, time=0, value=0.7}
         {interpMode=RCIM_Cubic, tangentMode=RCTM_User, time=0.8, value=0.4}
         {arriveTangent=2.7919, interpMode=RCIM_Cubic, leaveTangent=2.7919, tangentMode=RCTM_User, time=1, value=1.5}
        ]}}}
     CurveIndex=&RandomRangeFloat(Minimum=0, Maximum=1)
    )) {isSetParameters=true}
   &InitializeParticle(
    Color=&SelectLinearColorFromArray(
     Array Sampling Mode=::NewEnumerator0
     Direct Array Index=?
     Color Selection Array=DataInterface() {propertyValues {ColorData [<rgba 8 1.0354 0 1>, <rgba 10 0.6357 0 1>]}}
     Random Seed=0
    )
    Lifetime Max=6
    Lifetime Min=2
    Lifetime Mode=::NewEnumerator1
    Position Mode=::NewEnumerator1
    Sprite Rotation Mode=::NewEnumerator2
    Sprite Size Mode=::NewEnumerator1
    Uniform Sprite Size Max=25
    Uniform Sprite Size Min=20
    UsePositionOffset=false
    Write Material Random=true
   ) {module=InitializeParticle001}
   &ShapeLocation(MaskSpawnGroup=false, Override Seed=false, Rotation Axis=<0 0 1>, Rotation Mode=::NewEnumerator1, Sphere Radius=5000, UseOwnerScale=true)
   &AddVelocity(Velocity=<500 0 -500>)
   &SubUVAnimation(SubUV Animation Mode:ENiagara_SubUVLookupMode=::NewEnumerator2)
   &AddVelocity(Velocity=<450 -1250 0>) {module=AddVelocity001}
   &CurlNoiseForce(Noise Frequency=1, Noise Strength=2500, Pan Noise=false) {module=CurlNoiseForce002}
   &ApplyInitialForces(Apply Force to Position=true, Apply Force to Velocity=false)
  ]
  particleUpdateScript [
   &ParticleState()
   /Niagara/Modules/Update/Camera/MaintainInCameraParticleScale()
   &Drag(Drag=0, Use Linear Drag=true, Use Rotational Drag=false)
   &GravityForce()
   &CurlNoiseForce(
    Curl Noise Cone Mask Angle=120
    Mask Curl Noise=true
    Noise Frequency=&RandomRangeFloat(Minimum=2, Maximum=4)
    Noise Strength=&RandomRangeFloat(Minimum=1500, Maximum=2500)
    Pan Noise=true
    Pan Noise Field=<0 0 1>
   ) {module=CurlNoiseForce001}
   &SpriteRotationRate(Rotation Rate=&RandomRangeFloat(Minimum=-45, Maximum=45))
   &ScaleColor(
    Scale Alpha=&Lerp_Float(
     A=1
     B=0
     Alpha=&FloatFromCurve(
      FloatCurve=DataInterface() {propertyValues {Curve {keys [
          {arriveTangent=3.1144, interpMode=RCIM_Cubic, leaveTangent=3.1144, tangentMode=RCTM_User, time=0, value=0}
          {interpMode=RCIM_Cubic, tangentMode=RCTM_User, time=1, value=1}
         ]}}}
      Scale Curve=250
     )
    )
    Scale RGB=<5 5 1>
    ScaleA=false
    ScaleRGB=true
   )
   &ScaleSpriteSize(
    Uniform Curve Scale=3
    Uniform Curve Sprite Scale=DataInterface() {propertyValues {Curve {keys [
        {arriveTangent=0.0565, interpMode=RCIM_Cubic, leaveTangent=0.0565, tangentMode=RCTM_User, time=0, value=0}
        <at 0.2 0.9 RCIM_Cubic>
        {arriveTangent=-0.025, interpMode=RCIM_Cubic, leaveTangent=-0.025, tangentMode=RCTM_User, time=1, value=0}
       ]}}}
   ) {module=ScaleSpriteSize001}
   &DynamicMaterialParameters(
    Index 0 Param 1=&RandomRangeFloat(Minimum=0, Maximum=1)
    Index 0 Param 2=&RandomRangeFloat(Minimum=0, Maximum=1)
    Module.Param0WriteEnabled=true
    Module.Param1WriteEnabled=true
    Module.Param2WriteEnabled=true
    Module.Param3WriteEnabled=true
   ) {module=DynamicMaterialParameters002}
   &SolveForcesAndVelocity(Clamp Velocity=true, Limit Acceleration=false)
  ]
  rendererClasses [&NiagaraSpriteRendererProperties]
  renderers [{rendererClass=&NiagaraSpriteRendererProperties}]
 }}
summary {
 systemName=NS_Castle_FallingEmbers
 emitters [{emitterName=Embers_Burning_Cylinder, simTarget=GPUComputeSim, rendererClasses [&NiagaraSpriteRendererProperties]}]
 userVariables [
  User.CylinderHeight {type=NiagaraFloat, defaultValue=500, description=""}
  User.CylinderRadius {type=NiagaraFloat, defaultValue=7000, description=""}
  User.SpawnRate {type=NiagaraFloat, defaultValue=100, description=""}
 ]
}
