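# Niagara effect of smoke wisps for the top of the lava fall at the throne in the Helios castle environment.
# One GPU sprite emitter spawns steadily at either of two user offsets, picked by a random 0-or-1 lerp.
# Wisps drift under curl noise, point force and drag, growing and fading with randomized flipbook start frames.
# Color starts from a system lava color, is scaled orange, and runs through a time-of-day tint module.
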NiagaraSystem /Game/Environments/Helios/FX/Castle/Effects/NS_Castle_LavaFall_Top_Throne
dependencies {
 usedDataInterfaces [&NiagaraDataInterfaceCurve]
 usedDynamicInputs []
 usedModules [
  &self:SystemSpawnScript.SystemScriptSource.SystemScriptGraph.NiagaraNodeAssignment_0.SetVariables_#1
  &SystemState
 ]
}
emitters {Smoke_Wisp001 {
  simTarget=GPUComputeSim
  emitterUpdateScript [
   &EmitterState()
   &SpawnRate(SpawnRate=15, Use Spawn Probability=false)
  ]
  particleSpawnScript [
   &InitializeParticle(
    Color Maximum=$System.LavaColor
    Color Minimum=$System.LavaColor
    Color Mode=::NewEnumerator2
    Lifetime Max=2.25
    Lifetime Min=1.5
    Lifetime Mode=::NewEnumerator1
    Sprite Rotation Mode=::NewEnumerator2
    Sprite Size Mode=::NewEnumerator1
    Uniform Sprite Size Max=100
    Uniform Sprite Size Min=50
    UsePositionOffset=false
    Write Material Random=true
   ) {module=InitializeParticle001}
   &SystemLocation(
    MaskSpawnGroup=false
    Offset=&Lerp_Vector(
     A=$User.OffsetPos1
     B=$User.OffsetPos2
     Alpha=&MakeFloatFromInt(Int=&RandomRangeInt(Minimum=0, Maximum=1))
    )
   )
   /SetVariables(Particles.RandomFrame=&RandomRangeInt(Minimum=0, Maximum=63)) {isSetParameters=true}
   &DynamicMaterialParameters(
    Index 0 Param 1=&RandomRangeFloat(Minimum=0.2, Maximum=0.5)
    Module.Param0WriteEnabled=true
    Module.Param1WriteEnabled=true
    Module.Param2WriteEnabled=true
    Module.Param3WriteEnabled=true
   )
   &ShapeLocation(
    Cylinder Height=1
    Cylinder Radius=1
    MaskSpawnGroup=false
    Override Seed=false
    Shape Primitive=::NewEnumerator1
    Surface Only=false
    UseOwnerScale=true
   )
   &AddVelocity(Module.Use Falloff Distance=false, Origin Offset=<0 0 -100>, Override Seed=false, Velocity Mode=::NewEnumerator1)
   &CameraOffset(Camera Offset Amount=&RandomRangeFloat(Minimum=1, Maximum=10))
  ]
  particleUpdateScript [
   &CurlNoiseForce(Noise Frequency=2, Noise Strength=25, Pan Noise=false)
   &PointForce(Force Strength=25, Module.Use Falloff=false, Module.UseLinearForce=false)
   &Drag(Drag=&RandomRangeFloat(Minimum=1, Maximum=1), Use Linear Drag=true, Use Rotational Drag=false)
   &GravityForce(Gravity=&RandomRangeVector(Minimum=<0 0 0>, Maximum=<0 0 0>))
   &SolveForcesAndVelocity(Clamp Velocity=false, Limit Acceleration=false)
   &ParticleState()
   &FN_SubUVAnimation(
    End Frame=&Add_Integer(A=$Particles.RandomFrame, B=63)
    Start Frame=$Particles.RandomFrame
    SubUV Loop Count:NiagaraFloat=&RandomRangeFloat(Minimum=0.75, Maximum=1)
   )
   &ScaleSpriteSize(Scale Factor=&Vector2DFromFloat(Value=&FloatFromCurve(FloatCurve=DataInterface() {propertyValues {Curve {keys [
          {arriveTangent=2.5076, interpMode=RCIM_Cubic, leaveTangent=2.5076, tangentMode=RCTM_User, time=0.0023, value=0}
          {arriveTangent=0.0897, interpMode=RCIM_Cubic, leaveTangent=0.0897, tangentMode=RCTM_User, time=1, value=1}
         ]}}})))
   &RegionMask(Region Radius=25, UseOffset=false, UseRotation=false)
   &ScaleColor(
    Scale Alpha=&FloatFromCurve(
     FloatCurve=DataInterface() {propertyValues {Curve {keys [
         {arriveTangent=20.383, interpMode=RCIM_Cubic, leaveTangent=20.383, tangentMode=RCTM_User, time=0, value=0}
         {arriveTangent=0.3312, interpMode=RCIM_Cubic, leaveTangent=-2.6189, tangentMode=RCTM_Break, time=0.1744, value=0.9882}
         {arriveTangent=-0.5419, interpMode=RCIM_Cubic, leaveTangent=-0.5419, tangentMode=RCTM_User, time=1, value=0}
        ]}}}
     Scale Curve=0.25
    )
    Scale RGB=<1 0.5 0.2>
    ScaleA=true
    ScaleRGB=true
   ) {module=ScaleColor001}
   &DynamicMaterialParameters(
    Index 0 Param 1=&FloatFromCurve(
     FloatCurve=DataInterface() {propertyValues {Curve {keys [
         {arriveTangent=2.8718, interpMode=RCIM_Cubic, leaveTangent=2.8718, tangentMode=RCTM_User, time=0, value=0}
         {arriveTangent=-0.3942, interpMode=RCIM_Cubic, leaveTangent=-0.3942, tangentMode=RCTM_User, time=1, value=1}
        ]}}}
     Scale Curve=0
    )
    Module.Index0Param0Write=true
    Module.Index0Param1Write=true
    Module.Index0Param2Write=true
    Module.Index0Param3Write=true
   ) {module=DynamicMaterialParameters001}
   &self:ColorFromTOD_11(
    Afternoon:LinearColor=<rgba 1 0.9954 0.9249 1>
    Evening:LinearColor=<rgba 0.989 1 0.7298 1>
    Morning:LinearColor=<rgba 1 0.9526 0.8666 1>
    Night:LinearColor=<rgba 0.2955 0.5116 0.7 1>
    TOD:Vector4f=&MakeVector4_FromLinearColor(LinearColor=$NPC.FortniteNPC.FortniteActiveTimeOfDay)
   ) {module=ColorFromTOD_11}
  ]
  rendererClasses [&NiagaraSpriteRendererProperties]
  renderers [{rendererClass=&NiagaraSpriteRendererProperties}]
 }}
summary {
 systemName=NS_Castle_LavaFall_Top_Throne
 emitters [{emitterName=Smoke_Wisp001, simTarget=GPUComputeSim, rendererClasses [&NiagaraSpriteRendererProperties]}]
 userVariables [
  User.OffsetPos1 {type=Vector3f, description="", defaultValue {struct=Vector3f, value=<80 -65 60>}}
  User.OffsetPos2 {type=Vector3f, description="", defaultValue {struct=Vector3f, value=<-95 -80 50>}}
 ]
}
